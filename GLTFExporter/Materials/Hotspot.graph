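# Translucent unlit material for a hotspot sprite in the GLTFExporter plugin, with depth testing disabled so it draws over scene geometry.
# A Sprite texture parameter, defaulting to the engine's editor Anchor texture, supplies an alpha channel.
# An OpacityMask scalar lerps from zero to that alpha, so it scales the sprite's alpha and passes it unchanged at its default.
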Material /GLTFExporter/Materials/Hotspot
graph {
 properties {
  bCastRayTracedShadows=false
  bDisableDepthTest=true
  blendMode=BLEND_Translucent
  pixelDepthOffsetMode=PDOM_Legacy
  shadingModel=MSM_Unlit
  usedShadingModels=Unlit
 }
 nodes [LinearInterpolate(
   A=Constant()
   B=TextureSampleParameter2D() {parameterName=Sprite, texture=/Engine/EditorMaterials/Anchor}.A
   Alpha=ScalarParameter() {defaultValue=1, parameterName=OpacityMask, sliderMax=1}
  )]
}
