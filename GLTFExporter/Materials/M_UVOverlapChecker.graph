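# Unlit additive material in the glTF Exporter plugin, named as a checker for overlapping UVs.
# Its graph holds only a white color constant and a one-half scalar, with no textures or parameters.
# Additive blending makes overlapping draws accumulate brightness, which suits revealing overlap.
# It is flagged for use with skeletal meshes and mesh deformers, and casts no ray-traced shadows.
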Material /GLTFExporter/Materials/M_UVOverlapChecker
graph {
 properties {
  bCastRayTracedShadows=false
  bUseTranslucencyVertexFog=false
  bUsedWithMeshDeformer=true
  bUsedWithSkeletalMesh=true
  blendMode=BLEND_Additive
  pixelDepthOffsetMode=PDOM_Legacy
  shadingModel=MSM_Unlit
  usedShadingModels=Unlit
 }
 nodes [
  Constant3Vector() {constant=<rgba 1 1 1 0>}
  Constant() {r=0.5}
 ]
}
